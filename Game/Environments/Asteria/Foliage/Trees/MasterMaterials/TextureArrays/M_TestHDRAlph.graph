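# Test material for inspecting texel values in baked tree animation textures, drawn as on-screen numbers with debug value functions.
# SamplePOS_X/Y select a texel centre on a 64x26 grid, sampled from a Birch orientation bake and an Alder wind position bake.
# The Birch value is remapped from 0..1 to -1..1, and both colour-coded readouts are added.
# No material outputs are connected; a texture array sample and exposure multiply are loose experiments.
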Material /Game/Environments/Asteria/Foliage/Trees/MasterMaterials/TextureArrays/M_TestHDRAlph
graph {
 properties {pixelDepthOffsetMode=PDOM_Legacy, usedShadingModels=Default Lit}
 nodes [
  n1 = Divide(A=Constant2Vector() {g=1, r=1}, B=Constant2Vector() {g=26, r=64})
  n2 = ScalarParameter() {defaultValue=440.8238, parameterName=SamplePOS_X}
  n3 = ScalarParameter() {defaultValue=10957.4316, parameterName=SamplePOS_Y}
  n4 = Add(A=Multiply(A=@n1, B=AppendVector(A=@n2, B=@n3)), B=Multiply(A=@n1) {constB=0.5})
  n5 = TextureSample(UVs=@n4) {samplerType=SAMPLERTYPE_Masks, texture=/Game/Environments/Asteria/Foliage/Trees/Birch/BakeData/Birch8Anim_SKEL_ANIM_Position_ORIENT_V01}
  n6 = Constant() {desc=Bias, r=0.5}
  n7 = Subtract(
   A=@n5.RGBA
   B=MaterialFunctionCall(X=@n6, Y=@n6, Z=@n6, A=@n6) {materialFunction=&MakeFloat4}.Result
  ) {constB=0.5}
  n8 = Constant() {desc=Range, r=2}
  n9 = MaterialFunctionCall(X=@n8, Y=@n8, Z=@n8, A=@n8) {materialFunction=&MakeFloat4}
  n10 = ScalarParameter() {defaultValue=0.128, parameterName=UVoffset}
  n11 = AppendVector(A=@n10, B=@n10)
  n12 = ScalarParameter() {defaultValue=0.496, parameterName=UVoffset2}
  n13 = AppendVector(A=@n12, B=@n12)
  n14 = TextureSample(UVs=@n4) {
   samplerType=SAMPLERTYPE_LinearColor
   texture=/Game/Environments/Asteria/Foliage/Trees/FallTree/BakeData/LargeAlder_SKEL_ANIM_Wind_POSITION_V01
  }
  Add(
   A=MaterialFunctionCall!(
    Vector4=Multiply(A=@n7, B=@n9.Result) {constB=2}
    MaximumNumberOfDigits
    UVs
    DebugTextLocation RG_UpperRight BA_LowerLeft=AppendVector(A=@n11, B=@n13)
    Component Spacing
   ) {materialFunction=/Engine/Functions/Engine_MaterialFunctions02/Utility/DebugFloat4Values}.ColorCodedOutput
   B=MaterialFunctionCall(Vector4=@n14.RGBA) {materialFunction=&DebugFloat4Values}.ColorCodedOutput
  )
  MaterialFunctionCall() {materialFunction=&DebugScalarValues}
  MaterialFunctionCall!(Vector3=@n14.RGB, MaximumNumberOfDigits, UVs, DebugTextLocation RG_UpperRight BA_LowerLeft, Component Spacing) {materialFunction=/Engine/Functions/Engine_MaterialFunctions02/Utility/DebugFloat3Values}
  Multiply(B=ScalarParameter() {defaultValue=-213.6125, parameterName=Exposure})
  TextureSampleParameter2DArray!(UVs=AppendVector(A=TextureCoordinate(), B=ScalarParameter() {parameterName=index}), Apply View MipBias) {
   parameterName=arraye
   samplerType=SAMPLERTYPE_LinearColor
   texture=&/TestHDRAlpha
  }
  TextureSample() {
   samplerType=SAMPLERTYPE_LinearColor
   texture=/Game/Environments/Asteria/Foliage/Trees/Oak/Textures/Animation/MediumAlder/MediumAlder_SKEL_ANIM_Position_ORIENT_V01
  }
 ]
}
